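# Physics asset for the Crazy Eight medium male backpack cosmetic's skeletal mesh.
# It defines a single capsule body on the pack bone, offset and rotated to approximate the pack's shape.
# There are no constraints, so it provides one simple collision volume rather than jointed or secondary physics motion.
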PhysicsAsset /BRCosmetics/Accessories/FORT_Backpacks/Backpack_M_MED_Crazy_Eight/Meshes/M_MED_Crazy_Eight_Pack_PhysicsAsset
bodies {pack [{shapeType=Capsule, length=10.8357, radius=10.8243, center=<-0.001 -9.7138 -4.1554>, rotation=<rot -5.937 -90 0.3159>}]}
constraints []
